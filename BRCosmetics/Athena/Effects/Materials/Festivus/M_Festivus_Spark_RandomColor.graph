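# Translucent unlit sprite material for Festivus spark particles, drawing a soft flare mask tinted by particle color with a per-particle random hue shift.
# UVs are remapped from particle size, adjusted for camera-to-particle direction, keeping round caps on stretched sprites ("hotdogs").
# The tinted mask feeds an emissive adjustment function, while music-beat collection parameters, clamped, scale particle alpha.
# Mobile paths skip the random hue and UV remap.
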Material /BRCosmetics/Athena/Effects/Materials/Festivus/M_Festivus_Spark_RandomColor
graph {
 properties {
  bUsedWithParticleSprites=true
  blendMode=BLEND_Translucent
  pixelDepthOffsetMode=PDOM_Legacy
  shadingModel=MSM_Unlit
  usedShadingModels=Unlit
 }
 nodes [
  n1 = Multiply(
   A=ShadingPathSwitch(Default=ParticleRandom(), Mobile=Constant()) {desc=Migrated from FeatureLevelSwitch}
   B=ScalarParameter() {defaultValue=0.5, parameterName=RandomAmount}
  )
  Add(A=@n1) {constB=0.1}
  n2 = ParticleColor()
  n3 = Constant2Vector() {g=0.1, r=1}
  n4 = Abs(None=DotProduct(A=CameraVectorWS(), B=ParticleDirection()))
  n5 = Normalize(VectorInput=Multiply(A=ParticleSize(), B=LinearInterpolate(B=@n3, Alpha=@n4) {constA=1}))
  n6 = ComponentMask(None=@n5) {g=true}
  n7 = ComponentMask(None=@n5) {r=true}
  n8 = Clamp(None=Multiply(A=Divide(A=@n6, B=@n7), B=@n6) {constB=2}, Min=@n7) {maxDefault=999}
  n9 = Multiply(A=AppendVector(A=Constant() {r=1}, B=@n8), B=TextureCoordinate())
  n10 = Add(A=TextureCoordinate() {vTiling=0}, B=Constant2Vector() {g=0.5})
  n11 = ComponentMask(None=Clamp(None=@n9)) {g=true}
  n12 = Add(A=@n9, B=AppendVector(A=Constant(), B=OneMinus(None=@n8)))
  n13 = Clamp(None=@n12)
  n14 = ShadingPathSwitch(
   Default=LinearInterpolate(A=LinearInterpolate(A=@n9, B=@n10, Alpha=@n11) {constAlpha=0, desc=Blend top and middle}, B=@n12, Alpha=ComponentMask(None=@n13) {g=true}) {desc=Blend bottom}
   Mobile=TextureCoordinate() {desc=No hotdogs for mobile}
  ) {desc=Migrated from FeatureLevelSwitch}
  MaterialFunctionCall(Emissive=Multiply(
    A=MaterialFunctionCall(Hue Shift Percentage=@n1, Texture=@n2.RGB) {materialFunction=&HueShift}.Result
    B=TextureSampleParameter2D(UVs=@n14) {parameterName=RadialGradient, texture=/Game/Global/Textures/Default/Masks_and_Gradients/T_New_Soft_Flair}.R
   )) {materialFunction=&MF_FX_EmissiveAdjustments}
  n15 = CollectionParameter() {collection=/Game/Athena/Environments/Festivus/MotorizedLights/Lighting_MPC, parameterName=1_Beat_1}
  n16 = Multiply(
   A=CollectionParameter() {collection=/Game/Athena/Environments/Festivus/Materials/FFT_Materials/FFT_Breakout_MCP, parameterName=FFT_01_Smothed}
   B=ScalarParameter() {defaultValue=5, parameterName=Beat_Intensity}
  )
  Multiply(A=Clamp(None=Add(A=@n15, B=@n16)) {minDefault=0.5}, B=@n2.A) {constB=150}
 ]
}
