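# Translucent unlit sparkle material for Festivus particle and Niagara sprites, with responsive AA and a pulse driven by the beat.
# Particle color RGB goes through the shared FX emissive adjustments function.
# Particle alpha masks a four-point star texture's green channel, scaled by a clamped beat value.
# That value adds a lighting collection's beat parameter to a smoothed FFT collection value weighted by Beat_Intensity.
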Material /BRCosmetics/Athena/Effects/Materials/Festivus/M_Festivus_Sparkles_Simple
graph {
 properties {
  bEnableResponsiveAA=true
  bUsedWithNiagaraSprites=true
  bUsedWithParticleSprites=true
  blendMode=BLEND_Translucent
  pixelDepthOffsetMode=PDOM_Legacy
  shadingModel=MSM_Unlit
  usedShadingModels=Unlit
 }
 nodes [
  MaterialFunctionCall(Emissive=ParticleColor().RGB) {materialFunction=&MF_FX_EmissiveAdjustments}
  n1 = CollectionParameter() {collection=/Game/Athena/Environments/Festivus/MotorizedLights/Lighting_MPC, parameterName=1_Beat_1}
  n2 = Multiply(
   A=CollectionParameter() {collection=/Game/Athena/Environments/Festivus/Materials/FFT_Materials/FFT_Breakout_MCP, parameterName=FFT_01_Smothed}
   B=ScalarParameter() {defaultValue=1, parameterName=Beat_Intensity}
  )
  Multiply(
   A=Multiply(A=ParticleColor().A, B=TextureSample() {texture=/Game/Effects/Fort_Effects/Textures/Sparks/T_Sparkles_4Star_01_D}.G)
   B=Clamp(None=Add(A=@n1, B=@n2)) {maxDefault=2500, minDefault=0.5}
  ) {constB=150}
 ]
}
